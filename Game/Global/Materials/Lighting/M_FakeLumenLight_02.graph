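# Unlit, Nanite-enabled material for fake light surfaces, named for Lumen lighting, with ray-traced shadows disabled.
# A static switch picks a normalized blackbody temperature color (default) or a Color_Time_of_Day blend that defaults to warm at morning, evening and night and black by day.
# Brightness scales the color, and rounded V applies LowerBrightness to one UV half.
# Both brightness values use custom primitive data for per-primitive control.
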Material /Game/Global/Materials/Lighting/M_FakeLumenLight_02
graph {
 properties {bCastRayTracedShadows=false, bUsedWithNanite=true, pixelDepthOffsetMode=PDOM_Legacy, shadingModel=MSM_Unlit, usedShadingModels=Unlit}
 nodes [
  n1 = Normalize(VectorInput=BlackBody!(Temp=ScalarParameter() {defaultValue=1900, parameterName=Temp}))
  n2 = ScalarParameter() {defaultValue=1, parameterName=Brightness, primitiveDataIndex=1}
  n3 = VectorParameter() {parameterName=Color Night, channelNames {a=Alpha, b=Blue, g=Green, r=Red}, defaultValue=<rgba 1 0.4941 0.1198 1>}
  n4 = VectorParameter() {parameterName=Color Day, channelNames {a=Alpha, b=Blue, g=Green, r=Red}, defaultValue=<rgba 0 0 0 1>}
  n5 = MaterialFunctionCall(Color_Morning=@n3, Color_Day=@n4, Color_Evening=@n3, Color_Night=@n3) {materialFunction=&Color_Time_of_Day}
  n6 = Round(None=ComponentMask(None=TextureCoordinate()) {g=true})
  Multiply(
   A=StaticSwitchParameter(True=Multiply(A=@n1, B=@n2), False=Multiply(A=@n5.Result, B=@n2)) {defaultValue=true, parameterName=UseTemperature}
   B=LinearInterpolate(B=ScalarParameter() {parameterName=LowerBrightness, primitiveDataIndex=2}, Alpha=Saturate(None=@n6)) {constA=1}
  )
 ]
}
