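# Unlit material in the global lighting set that produces an animated, turbulent HDR pink-red glow.
# A slowly panning, stretched fluid-flame texture's green channel drives a lerp from near black to full strength, modulating a bright color parameter.
# The Portal_Intensity scalar from the FortniteMaterialParameters collection multiplies the result, so brightness follows that shared global value.
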Material /Game/Global/Materials/Lighting/M_TurbulentLumen
graph {
 properties {bAutomaticallySetUsageInEditor=false, shadingModel=MSM_Unlit, usedShadingModels=Unlit}
 nodes [
  n1 = TextureSample(UVs=Panner(Coordinate=TextureCoordinate() {uTiling=0.2, vTiling=0.1}) {speedX=0.05, speedY=0.1}) {samplerType=SAMPLERTYPE_LinearGrayscale, texture=/Game/Weapons/FORT_Melee/Pickaxe_Cube_Ninja_Male/FX/T_FluidFlameTile}
  n2 = CollectionParameter() {collection=&FortniteMaterialParameters, parameterName=Portal_Intensity}
  Multiply(A=Multiply(A=VectorParameter() {defaultValue=<rgba 200 5.2083 75.2508 1>}.RGB, B=LinearInterpolate(Alpha=@n1.G) {constA=0.0002}), B=@n2)
 ]
}
